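; Looping gameplay cue for the SunRose SilkyWolf ready state, presenting the player's skull charge count through Niagara effects and audio.
; On start it reads the charges, then an async listener on CurrentCharges updates the Niagara count and loop audio parameter, with activation or deactivation sounds.
; Gameplay tag tasks track targeting and DBNO as bool parameters for the skull effect.
; Removal ends the stored attribute listener.
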(event Custom|Update%20Charge%20Count
  ;[0.0361 1 0] On Looping Start
  ;;[0.2578 0.9402 1] Listen for Counter count changes.
  (if (ScalableFloat|GetValueasBool (Variables|Default|GetShouldDelayChargeCount))
    (Utilities|FlowControl|DelayUntilNextTick 0))
  (bind _playerpawn (Variables|Default|GetPlayerPawn)
    (:then
      (CallFunction|OnChargeCountChanged (Math|Float|Truncate (GadgetAttributeSets|GetSimulatedAttributeValue _playerpawn (Variables|Default|GetChargeAttribute) 0.0 :!class /Script/FortniteGame.FortPlayerPawn))))
    (:else)))

(event EventOnLoopingStart(EitherVFX-preferthisversion)
  (params (MyTarget :type Actor%20Object%20Reference) (Parameters :type Gameplay%20Cue%20Parameters%20Structure%20%28by%20ref%29) (ParticleComponents :type Array%20of%20FXSystem%20Component%20Object%20References) (AudioComponents :type Array%20of%20Audio%20Component%20Object%20References))
  ;[0.0361 1 0] On Looping Start
  (|CallParentFunction MyTarget Parameters ParticleComponents AudioComponents)
  (bind _asfort_player_pawn (Utilities|Casting|CastToFortPlayerPawn MyTarget))
  (Variables|Default|SetPlayerPawn _asfort_player_pawn)
  (Variables|Default|SetNS_Skulls (Utilities|Array|Get(acopy) ParticleComponents 1))
  (Utilities|FlowControl|Sequence
    (:then_0
      (CallFunction|SetupPlayerTargetingBool))
    (:then_1
      (CallFunction|SetupDBNOBool))
    (:then_2
      (Variables|Default|SetLooping%20Audio%20Component (Utilities|Array|Get(acopy) AudioComponents))
      (Utilities|FlowControl|Sequence
        (:then_0
          ;;[0.2664 1 0.1888] Initialize Charge Count
          (bind _total_charges (Class|FunctionLibrarySunRoseTokenSilkyWolf|GetTotalCharges (Ability|GetAbilitySystemComponent (Variables|Default|GetPlayerPawn)) (Ability|GetAbilitySystemComponent (Variables|Default|GetPlayerPawn))))
          (Variables|Default|SetCharge%20Count _total_charges)
          (CallFunction|PlayActivationSound))
        (:then_1
          ;;[0.2578 0.9402 1] Listen for Counter count changes.
          (bind _asynctaskproxy (Ability|Tasks|WaitForSimulatedAttributeChanged (Variables|Default|GetPlayerPawn) "(AttributeName=\"CurrentCharges\",Attribute=/Script/SunRoseEnvGameplayRuntime.FortChargingSet_SunroseDash:CurrentCharges,AttributeOwner=\"/Script/CoreUObject.Class'/Script/SunRoseEnvGameplayRuntime.FortChargingSet_SunroseDash'\")")
            (:then
              (Utilities|Array|Add (Variables|Default|GetAsync%20Actions) _asynctaskproxy))
            (:Changed
              (CallFunction|UpdateChargeCount))))
        (:then_2
          (if true
            (Utilities|Array|Clear (Variables|Default|GetChargeFXs))
            (for _array_element ParticleComponents
              (bind _asniagara_particle_system_component (Utilities|Casting|CastToNiagaraComponent _array_element))
              (Utilities|Array|Add (Variables|Default|GetChargeFXs) _asniagara_particle_system_component)
              (Rendering|SetTranslucentSortPriority _asniagara_particle_system_component 10)
              (Niagara|SetNiagaraVariable(Int32) _asniagara_particle_system_component "NumCharges" (Math|Integer|Max(Integer) (Variables|Default|GetCharge%20Count) (Math|Float|Truncate (GadgetAttributeSets|GetSimulatedAttributeValue (Variables|Default|GetPlayerPawn) (Variables|Default|GetChargeAttribute) 0.0 :!class /Script/FortniteGame.FortPlayerPawn)))))))
        (:then_3
          (CallFunction|UpdateAudio))))))

(event EventOnRemoval(EitherVFX-preferthisversion)
  (params (MyTarget :type Actor%20Object%20Reference) (Parameters :type Gameplay%20Cue%20Parameters%20Structure%20%28by%20ref%29) (ParticleComponents :type Array%20of%20FXSystem%20Component%20Object%20References) (AudioComponents :type Array%20of%20Audio%20Component%20Object%20References) (BurstCameraShakeInstance :type Camera%20Shake%20Base%20Object%20Reference) (BurstDecalInstance :type Decal%20Actor%20Object%20Reference))
  (bind _async_actions (Variables|Default|GetAsync%20Actions))
  ;[1 0.5026 0] On Removal
  (|CallParentFunction MyTarget Parameters ParticleComponents AudioComponents BurstCameraShakeInstance BurstDecalInstance)
  (for _array_element _async_actions
    (Ability|Async|EndAction _array_element))
  (Utilities|Array|Clear _async_actions))

(event Custom|Update%20Audio
  ;[0.07997 0.1892 0.2292] Update Audio Loop
  (Audio|Parameter|SetIntegerParameter (Variables|Default|GetLooping%20Audio%20Component) "FromGCN.ChargeCount" (Variables|Default|GetCharge%20Count) :!class /Script/AudioExtensions.AudioParameterControllerInterface))

(event Custom|OnChargeCountChanged
  (params (New%20Charge%20Count :type Integer))
  ;[0.1562 1 0.9901] OnChargecountChanged
  (bind _output_get (Variables|Default|SetCharge%20Count%20Delta (- New%20Charge%20Count (Variables|Default|GetCharge%20Count))))
  (Variables|Default|SetCharge%20Count New%20Charge%20Count)
  (if (!= _output_get 0)
    (Utilities|FlowControl|Sequence
      (:then_0
        (Development|PrintString (Variables|Default|GetMyPlayerController) (Utilities|String|Append "Skull Charge Count " (Utilities|String|ToString(Integer) (Variables|Default|GetCharge%20Count))))
        ;;[0.1406 0.02856 0.02856] Audio
        (CallFunction|UpdateAudio)
        (if (> (Variables|Default|GetCharge%20Count%20Delta) 0)
          (CallFunction|PlayActivationSound)
          (elif (<= (Variables|Default|GetCharge%20Count) 0)
            (CallFunction|PlayDeactivationSound))))
      (:then_1
        (for _array_element (Variables|Default|GetChargeFXs)
          (Niagara|SetNiagaraVariable(Int32) _array_element "NumCharges" (Variables|Default|GetCharge%20Count)))))))

(event Custom|Setup%20Player%20Targeting%20Bool
  (bind _playerpawn (Variables|Default|GetPlayerPawn))
  ;[1 1 1] Setup ADS changed Bool
  (Ability|Tasks|WaitGameplayTagAddToActor _playerpawn "(TagName=\"Gameplay.Status.IsTargeting\")"
    (:then
      (Ability|Tasks|WaitGameplayTagRemoveFromActor _playerpawn "(TagName=\"Gameplay.Status.IsTargeting\")"
        (:Removed
          (Variables|Default|SetIs%20ADS false)
          (Development|PrintString 0 "not ADS"))))
    (:Added
      (Variables|Default|SetIs%20ADS true)
      (Development|PrintString 0 "ADS")))
  (CallFunction|ADSChanged))

(event Custom|ADS%20Changed
  ;[1 0.624 0.004132] ADS Changed
  (Utilities|Casting|CastToFortPlayerController (Pawn|GetLocalViewingPlayerController (Variables|Default|GetPlayerPawn)))
  (Effects|Components|ParticleSystem|SetBoolParameter (Variables|Default|GetNS_Skulls) "blsADS" (Variables|Default|GetIs%20ADS)))

(event Custom|Setup%20DBNO%20Bool
  (bind _playerpawn (Variables|Default|GetPlayerPawn))
  ;[1 1 1] Setup DBNO Bool
  (Ability|Tasks|WaitGameplayTagAddToActor _playerpawn "(TagName=\"Gameplay.Status.DBNO\")"
    (:then
      (Ability|Tasks|WaitGameplayTagRemoveFromActor _playerpawn "(TagName=\"Gameplay.Status.DBNO\")"
        (:Removed
          (Variables|Default|SetIs%20DBNO false)
          (Development|PrintString 0 "Not DBNO"))))
    (:Added
      (Variables|Default|SetIs%20DBNO true)
      (Development|PrintString 0 "DBNO")))
  (CallFunction|DBNOChanged))

(event Custom|DBNO%20Changed
  ;[1 0.624 0.004132] DBNO Changed
  (Effects|Components|ParticleSystem|SetBoolParameter (Variables|Default|GetNS_Skulls) "blsDBNO" (Variables|Default|GetIs%20DBNO)))

(fn PlayActivationSound
  (bind _returnvalue (Audio|SpawnSoundAttached (Variables|Default|GetActivation%20Sound) (Variables|Default|GetLooping%20Audio%20Component)))
  (Audio|Parameter|SetIntegerParameter _returnvalue "FromGCNL.ChargeCount" (Variables|Default|GetCharge%20Count) :!class /Script/AudioExtensions.AudioParameterControllerInterface))

(fn PlayDeactivationSound
  (Audio|SpawnSoundAttached (Variables|Default|GetDeactivation%20Sound) (Variables|Default|GetLooping%20Audio%20Component)))

(fn ConstructionScript)
